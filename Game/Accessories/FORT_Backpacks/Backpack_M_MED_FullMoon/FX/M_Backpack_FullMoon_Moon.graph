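# Masked, unlit, two-sided Niagara particle material that draws the moon for the Full Moon backpack effect.
# A parallax bump-offset, mip-biased moon texture is tinted by particle color, then blended through panning noise with a copy scaled by the "Hue Shiuft" dynamic parameter for emissive.
# A circular mask times particle alpha is cut by a second sphere mask of dynamic radius, allowing crescent shapes.
# The cutout uses X_Position for both axes, leaving Y_Position unused.
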Material /Game/Accessories/FORT_Backpacks/Backpack_M_MED_FullMoon/FX/M_Backpack_FullMoon_Moon
graph {
 properties {
  bUsedWithNiagaraMeshParticles=true
  bUsedWithNiagaraSprites=true
  blendMode=BLEND_Masked
  pixelDepthOffsetMode=PDOM_Legacy
  shadingModel=MSM_Unlit
  twoSided=true
  usedShadingModels=Unlit
 }
 nodes [
  n1 = DynamicParameter() {defaultValue=<rgba 1 0 0 0>, paramNames [Hue Shiuft, X_Position, Y_Position, Radius]}
  n2 = ScalarParameter() {defaultValue=0.0921, parameterName=Height Texture}
  n3 = TextureCoordinate()
  n4 = ScalarParameter() {parameterName=Ref Plane}
  n5 = ScalarParameter() {defaultValue=0.25, parameterName="HeightRatioInput (S)"}
  n6 = ParticleColor()
  n7 = Multiply(
   A=TextureSample!(
    UVs=MaterialFunctionCall!(HeightTexture=@n2, TextureCoordinates=@n3, ReferencePlane=@n4, HeightRatioInput=@n5, CameraVector) {materialFunction=/Engine/Functions/Engine_MaterialFunctions02/Utility/BumpOffset_Advanced.BumpOffset_advanced}.UVs
    Tex
    Bias=ScalarParameter() {defaultValue=1, parameterName=Bias}
    Apply View MipBias
   ) {mipValueMode=TMVM_MipBias, texture=/Game/Packages/Fortress_Sky/MaterialInstance/Arid/T_Moon}.RGB
   B=@n6.RGB
  )
  MaterialFunctionCall(Hue Shift Percentage=@n1.Hue Shiuft, Texture=@n7) {materialFunction=&HueShift}
  n8 = Panner() {speedX=0.08}
  MaterialFunctionCall(Emissive=LinearInterpolate(
    A=Multiply(A=@n7, B=@n1.Hue Shiuft)
    B=@n7
    Alpha=TextureSample!(UVs=@n8, Tex, Apply View MipBias) {texture=&T_NoisePack_01}.G
   )) {materialFunction=&MF_FX_EmissiveAdjustments}
  n9 = TextureCoordinate()
  n10 = Constant() {r=0.5}
  n11 = TextureCoordinate()
  n12 = DynamicParameter() {defaultValue=<rgba 1 0 0 0>, paramNames [Hue Shiuft, X_Position, Y_Position, Radius]}
  n13 = AppendVector(A=@n12.X_Position, B=@n12.X_Position)
  Multiply(A=Multiply(A=@n6.A, B=SphereMask(A=@n9, B=@n10, Radius=@n10)), B=OneMinus(None=SphereMask(A=@n11, B=@n13, Radius=@n12.Radius)))
 ]
}
